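# Translucent, unlit, two-sided Niagara mesh-particle material for the Armadillo pickaxe's sword arc, drawing an electric, dissolving effect.
# Swirl-distorted panning lightning plus a gradient, masked by noise, blends two colors tinted by particle color.
# Opacity erodes that lightning with cell noise biased along V, driven by a dynamic parameter, with depth fade except on mobile.
# World position offset combines noise displacement along normals with noise-masked SplineThicken, its width scaled per particle.
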Material /BRCosmetics/Weapons/FORT_Melee/Pickaxe_Armadillo_Male/FX/Materials/Parents/M_Armadillo_Sword_Arc_01
graph {
 properties {
  bAllowNegativeEmissiveColor=true
  bAutomaticallySetUsageInEditor=false
  bUsedWithNiagaraMeshParticles=true
  bUsedWithStaticMesh=false
  blendMode=BLEND_Translucent
  pixelDepthOffsetMode=PDOM_Legacy
  shadingModel=MSM_Unlit
  twoSided=true
  usedShadingModels=Unlit
 }
 nodes [
  n1 = TextureCoordinate()
  n2 = VectorParameter() {
   group=Distortion
   parameterName=WPO_Sub_UV
   channelNames {a=WPO_Sub_V_Speed, b=WPO_Sub_U_Speed, g=WPO_Sub_V_Tile, r=WPO_Sub_U_Tile}
   defaultValue=<rgba 1 4 0.5 1>
  }
  n3 = AppendVector(A=@n2.WPO_Sub_U_Tile, B=@n2.WPO_Sub_U_Tile)
  n4 = TextureSampleParameter2D(UVs=Panner(Coordinate=Multiply(A=@n1, B=@n3), Speed=AppendVector(A=@n2.WPO_Sub_U_Speed, B=@n2.WPO_Sub_U_Speed))) {group=AlphaNoise, parameterName=WPOSubTex, samplerType=SAMPLERTYPE_Grayscale, texture=&LowResBlurredNoise}
  n5 = Multiply(
   A=Power(Base=@n4.G, Exp=ScalarParameter() {defaultValue=3, group=AlphaNoise, parameterName=WPO_Sub_Power})
   B=ScalarParameter() {defaultValue=8, group=AlphaNoise, parameterName=WPO_Sub_Intensity}
  )
  n6 = TextureCoordinate()
  n7 = VectorParameter() {
   group=Distortion
   parameterName=WPO_UV
   channelNames {a=WPO_V_Speed, b=WPO_U_Speed, g=WPO_V_Tile, r=WPO_U_Tile}
   defaultValue=<rgba 0.25 0.25 -0.05 0>
  }
  n8 = AppendVector(A=@n7.WPO_U_Tile, B=@n7.WPO_U_Tile)
  n9 = DynamicParameter() {paramNames [Random Offset, Lightning Offset, WPO Offset, Spline Width Mul]}
  n10 = Add(
   A=Panner(Coordinate=Multiply(A=@n6, B=@n8), Speed=AppendVector(A=@n7.WPO_U_Speed, B=@n7.WPO_U_Speed))
   B=AppendVector(A=ScalarParameter() {group=Tech, parameterName=WPO_U_Offset}, B=Add(A=@n9.Random Offset, B=@n9.Random Offset))
  )
  n11 = Multiply(
   A=Power(
    Base=TextureSampleParameter2D(UVs=@n10) {group=AlphaNoise, parameterName=WPOTex, texture=/Game/Global/Textures/Default/Noise/TankMoteNoise}.G
    Exp=ScalarParameter() {defaultValue=1, group=AlphaNoise, parameterName=WPO_Power}
   )
   B=ScalarParameter() {defaultValue=5, group=AlphaNoise, parameterName=WPO_Intensity}
  )
  n12 = TextureCoordinate()
  n13 = VectorParameter() {
   group=Distortion
   parameterName=Noise_UV
   channelNames {a=Lightning_V_Speed, b=Lightning_U_Speed, g=Lightning_V_Tile, r=Lightning_U_Tile}
   defaultValue=<rgba 1 1 0 0.5>
  }
  n14 = AppendVector(A=@n13.Lightning_U_Tile, B=@n13.Lightning_U_Tile)
  n15 = TextureSampleParameter2D(UVs=Panner(Coordinate=Multiply(A=@n12, B=@n14), Speed=AppendVector(A=@n13.Lightning_U_Speed, B=@n13.Lightning_U_Speed))) {group=AlphaNoise, parameterName=NoiseTex, samplerType=SAMPLERTYPE_Grayscale, texture=&LowResBlurredNoise}
  n16 = ScalarParameter() {defaultValue=1.5, group=AlphaNoise, parameterName=Noise_Power}
  n17 = Saturate(None=Multiply(A=Power(Base=@n15.G, Exp=@n16), B=ScalarParameter() {defaultValue=20, group=AlphaNoise, parameterName=Noise_Hardness}))
  n18 = DynamicParameter() {paramNames [Random Offset, Lightning Offset, WPO Offset, Spline Width Mul]}
  n19 = Multiply(A=ScalarParameter() {defaultValue=5, parameterName=WidthBase}, B=@n18.Spline Width Mul)
  n20 = Multiply(A=ScalarParameter() {defaultValue=2, parameterName="WidthTip (S)"}, B=@n18.Spline Width Mul)
  Add(
   A=Multiply(A=VertexNormalWS(), B=LinearInterpolate(A=@n5, B=@n11))
   B=Multiply(
    A=@n17
    B=MaterialFunctionCall(WidthBase=@n19, WidthTip=@n20) {materialFunction=&SplineThicken}.WorldPositionOffset
   )
  )
  n21 = VectorParameter() {parameterName=ColorA, defaultValue=<rgba 0.2 0 0.0167 1>}
  n22 = VectorParameter() {parameterName=ColorB, defaultValue=<rgba 10 0.3 0.6233 1>}
  n23 = TextureSample(UVs=MaterialFunctionCall(UVs=TextureCoordinate() {vTiling=0.5}, "Rotation Angle (0-1)"=Constant() {r=0.25}) {materialFunction=&CustomRotator}.Rotated Values) {samplerType=SAMPLERTYPE_Masks, texture=/Game/Global/Textures/Default/Masks_and_Gradients/T_VertGradient_MASK_2}
  n24 = VectorParameter() {
   group=Distortion
   parameterName=Distortion_UV
   channelNames {a=Distortion_Sub_V_Speed, b=Distortion_Sub_U_Speed, g=Distortion_Sub_V_Tile, r=Distortion_Sub_U_Tile}
   defaultValue=<rgba 1 2 -0.3 1>
  }
  n25 = Panner(
   Coordinate=Multiply(A=TextureCoordinate(), B=AppendVector(A=@n24.Distortion_Sub_U_Tile, B=@n24.Distortion_Sub_U_Tile))
   Speed=AppendVector(A=@n24.Distortion_Sub_U_Speed, B=@n24.Distortion_Sub_U_Speed)
  ) {speedY=0.3}
  n26 = Multiply(
   A=ComponentMask(None=TextureSampleParameter2D(UVs=@n25) {parameterName=DistortionTex, samplerType=SAMPLERTYPE_Normal, texture=/Game/Global/Textures/Default/Noise/T_SwirlNoise_N}.RGB) {g=true, r=true}
   B=ScalarParameter() {defaultValue=0.05, parameterName=Distortion_Intensity}
  ) {constB=0.1}
  n27 = TextureCoordinate()
  n28 = VectorParameter() {
   group=Distortion
   parameterName=Lightning_UV
   channelNames {a=Lightning_V_Speed, b=Lightning_U_Speed, g=Lightning_V_Tile, r=Lightning_U_Tile}
   defaultValue=<rgba 1 1 0 0.5>
  }
  n29 = AppendVector(A=@n28.Lightning_U_Tile, B=@n28.Lightning_U_Tile)
  n30 = DynamicParameter() {paramNames [Random Offset, Lightning Offset, WPO Offset, Spline Width Mul]}
  n31 = Add(
   A=Panner(Coordinate=Multiply(A=@n27, B=@n29), Speed=AppendVector(A=@n28.Lightning_U_Speed, B=@n28.Lightning_U_Speed))
   B=AppendVector(A=ScalarParameter() {group=Tech, parameterName=Lightning_U_Offset}, B=Add(A=@n30.Random Offset, B=@n30.Random Offset))
  )
  n32 = Add(
   A=Multiply(A=Power(Base=@n23.R) {constExponent=1.5}) {constB=0.2}
   B=TextureSampleParameter2D(UVs=Add(A=@n26, B=@n31)) {parameterName=LightningTex, texture=/Game/Athena/SafeZone/Textures/T_JanusShield_Lightning}.G
  )
  n33 = Saturate(None=Multiply(A=@n17, B=@n32))
  MaterialFunctionCall(Emissive=Multiply(A=LinearInterpolate(A=@n21.RGB, B=@n22.RGB, Alpha=@n33), B=ParticleColor().RGB)) {materialFunction=&MF_FX_EmissiveAdjustments}
  n34 = Multiply(A=@n33, B=@n33)
  n35 = Multiply(A=@n34, B=@n34)
  n36 = Multiply(A=@n35, B=@n35)
  Multiply(A=Multiply(A=@n36, B=@n36)) {constB=50}
  ScalarParameter() {group=Tech, parameterName=WPO_V_Offset}
  ScalarParameter() {group=Tech, parameterName=Lightning_V_Offset}
  n37 = VectorParameter() {
   group=Dissolve
   parameterName=Dissolve UV Control
   channelNames {a=Dissolve_V_PanSpeed, b=Dissolve_U_PanSpeed, g=Dissolve_V_Tilling, r=Dissolve_U_Tilling}
   defaultValue=<rgba 0.5 1 0 0.5>
  }
  n38 = Multiply(A=TextureCoordinate(), B=AppendVector(A=@n37.Dissolve_U_Tilling, B=@n37.Dissolve_U_Tilling))
  n39 = AppendVector(A=@n37.Dissolve_U_PanSpeed, B=@n37.Dissolve_U_PanSpeed)
  n40 = DynamicParameter() {paramNames [Random Offset, Lightning Offset, WPO Offset, Spline Width Mul]}
  n41 = Add(A=ScalarParameter() {group=Dissolve, parameterName=Dissolve_U_Offset}, B=@n40.Random Offset)
  n42 = Add(A=@n40.Random Offset, B=ScalarParameter() {group=Dissolve, parameterName=Dissolve_V_Offset})
  n43 = TextureSampleParameter2D(UVs=Add(A=Panner(Coordinate=@n38, Speed=@n39), B=AppendVector(A=@n41, B=@n42))) {
   group=Dissolve
   parameterName=Dissolve Tex
   samplerType=SAMPLERTYPE_Grayscale
   texture=/BRCosmetics/Weapons/FORT_Melee/Pickaxe_Cube_Ninja_Male/FX/T_CellNoise
  }
  n44 = LinearInterpolate(
   A=@n43.G
   B=ComponentMask(None=TextureCoordinate()) {g=true}
   Alpha=ScalarParameter() {defaultValue=0.75, group=Dissolve, parameterName=Lerp Direction}
  ) {constAlpha=0.75}
  n45 = ScalarParameter() {defaultValue=1, group=Dissolve, parameterName=Dissolve Power}
  n46 = ScalarParameter() {defaultValue=1, group=Dissolve, parameterName=Dissolve Edge Hardness, sliderMin=0.5}
  n47 = ScalarParameter() {defaultValue=1, group=Dissolve, parameterName=Dissolve Dynamic Test}
  n48 = DynamicParameter() {parameterIndex=1, paramNames [Dissolve, N/A, N/A, N/A]}
  n49 = Subtract(A=Multiply(A=Power(Base=@n44, Exp=@n45), B=@n46), B=LinearInterpolate(A=@n46, Alpha=Multiply(A=@n47, B=@n48.Dissolve)) {constB=-1})
  n50 = Multiply(A=Multiply(A=@n32, B=Saturate(None=@n49)), B=ScalarParameter() {defaultValue=2, parameterName=AlphaMul})
  n51 = DepthFade(Opacity=@n50, FadeDistance=ScalarParameter() {defaultValue=500, group=Depth Fade, parameterName=DepthFade})
  ShadingPathSwitch(Default=@n51, Deferred=@n51, Mobile=@n50) {desc=Migrated from FeatureLevelSwitch}
 ]
}
