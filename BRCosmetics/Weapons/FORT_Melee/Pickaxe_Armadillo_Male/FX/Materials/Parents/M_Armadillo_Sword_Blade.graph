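# Translucent unlit, two-sided Niagara particle material for the glowing blade effect of the Armadillo pickaxe sword.
# Dynamic-parameter-controlled UV masks (high, thin, polar), perturbed by panning distortion noise, are remapped through curve atlas rows to drive color bands and opacity.
# Light, dark and time-of-day-scaled core colors blend via electric noise, with dark noise detail and particle color.
# Opacity adds a cell-noise dissolve, optionally directional, plus depth fade except on mobile.
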Material /BRCosmetics/Weapons/FORT_Melee/Pickaxe_Armadillo_Male/FX/Materials/Parents/M_Armadillo_Sword_Blade
graph {
 properties {
  bUsedWithNiagaraMeshParticles=true
  bUsedWithNiagaraSprites=true
  blendMode=BLEND_Translucent
  pixelDepthOffsetMode=PDOM_Legacy
  shadingModel=MSM_Unlit
  twoSided=true
  usedShadingModels=Unlit
 }
 nodes [
  n1 = Power(Base=ComponentMask(None=TextureCoordinate()) {g=true}) {constExponent=50}
  n2 = Saturate(None=Saturate(None=Multiply(A=@n1)))
  n3 = Multiply(A=@n2) {constB=4}
  n4 = TextureCoordinate()
  n5 = ComponentMask(None=@n4) {g=true}
  n6 = DynamicParameter() {defaultValue=<rgba 0.65 0 0.45 0.3>, paramNames [Mask High, Mask Thin, Mask Polar, Distortion Intensity]}
  n7 = ComponentMask(None=@n4) {r=true}
  n8 = Multiply(A=Abs(None=Subtract(A=@n7) {constB=0.5})) {constB=2}
  n9 = Multiply(A=@n6.Mask Thin, B=@n8)
  n10 = TextureCoordinate()
  n11 = Constant() {r=0.5}
  n12 = LinearInterpolate(
   A=LinearInterpolate(A=Subtract(A=@n5, B=@n9), B=SphereMask(A=@n10, B=@n11) {attenuationRadius=0.5, hardnessPercent=0}, Alpha=@n6.Mask Polar)
   B=OneMinus(None=@n8)
   Alpha=@n6.Mask High
  )
  n13 = CurveAtlasRowParameter(CurveTime=Subtract(A=@n12, B=Multiply(A=@n3) {constB=2})) {
   atlas=/Game/Athena/LinearColorCurves/TempAthenaCurveAtlas
   curve=/Game/Gadgets/Assets/VinderTech_GliderChute/Glider_Donut_Plate/FX/Curve_Donut_Plate02
   defaultValue=6
   parameterName=Curve_DonutPlate_2
  }
  Add(B=@n13.R)
  n14 = VectorParameter() {group=Color, parameterName=Color_Light, defaultValue=<rgba 1 0.01 0.043 1>}
  n15 = Constant() {r=0.5}
  n16 = Multiply(
   A=VectorParameter() {group=Color, parameterName=Color_Core, defaultValue=<rgba 30 0.9 2.8399 1>}.RGB
   B=MaterialFunctionCall(Value_Morning=@n15, Value_Day=Constant() {r=1}, Value_Evening=@n15, Value_Night=Constant() {r=0.1}) {materialFunction=&Scalar_Time_of_Day}.Result
  )
  n17 = TextureCoordinate()
  n18 = VectorParameter() {
   group=Noise
   parameterName=Noise_UV
   channelNames {a=Noise_V_Speed, b=Noise_U_Speed, g=Noise_V_Tile, r=Noise_U_Tile}
   defaultValue=<rgba 1 1 0 0.5>
  }
  n19 = Multiply(A=@n17, B=AppendVector(A=@n18.Noise_U_Tile, B=@n18.Noise_U_Tile))
  n20 = AppendVector(A=@n18.Noise_U_Speed, B=@n18.Noise_U_Speed)
  n21 = VectorParameter() {
   group=Noise
   parameterName=Noise_Dist_UV
   channelNames {a=Noise_Dist_V_Speed, b=Noise_Dist_U_Speed, g=Noise_Dist_V_Tile, r=Noise_Dist_U_Tile}
   defaultValue=<rgba 1 1 0 2>
  }
  n22 = AppendVector(A=@n21.Noise_Dist_U_Tile, B=@n21.Noise_Dist_U_Tile)
  n23 = TextureSample(UVs=Panner(Coordinate=Multiply(A=@n17, B=@n22), Speed=AppendVector(A=@n21.Noise_Dist_U_Speed, B=@n21.Noise_Dist_U_Speed)) {speedY=1}) {samplerType=SAMPLERTYPE_LinearGrayscale, texture=/Game/Global/Textures/Default/Noise/T_SwirlNoise_N_height}
  n24 = ScalarParameter() {defaultValue=0.2, group=Noise, parameterName=Noise_Dist_Intensity}
  n25 = TextureSampleParameter2D(UVs=Add(A=Panner(Coordinate=@n19, Speed=@n20) {speedY=0.5}, B=Multiply(A=@n23.R, B=@n24) {constB=0.2})) {
   group=Noise
   parameterName=Noise_Tex
   samplerType=SAMPLERTYPE_LinearColor
   texture=/Game/Characters/Player/Male/Medium/Bodies/M_MED_SRDrift/Textures/T_ElectricNoise
  }
  n26 = OneMinus(None=SphereMask(A=TextureCoordinate(), B=Constant2Vector() {g=0.3, r=0.5}) {attenuationRadius=0.65, hardnessPercent=0})
  n27 = CurveAtlasRowParameter(CurveTime=@n26) {
   atlas=&TempAthenaCurveAtlas
   curve=&Curve_Donut_Plate02
   defaultValue=6
   parameterName=Curve_DonutPlate_2
  }
  n28 = Add(A=Multiply(A=Multiply(A=@n25.R, B=@n26), B=ScalarParameter() {defaultValue=2, group=Noise, parameterName=Noise_Intensity}), B=@n27.G)
  n29 = VectorParameter() {group=Color, parameterName=Color_Dark, defaultValue=<rgba 0.5 0 0.0417 1>}
  n30 = CurveAtlasRowParameter(CurveTime=@n12) {
   atlas=&TempAthenaCurveAtlas
   curve=/Game/Gadgets/Assets/VinderTech_GliderChute/Glider_Donut_Plate/FX/Curve_Donut_Plate01
   defaultValue=5
  }
  n31 = Subtract(A=@n30.B, B=@n3) {constB=0}
  n32 = ScalarParameter() {defaultValue=0.1, group=Distortion, parameterName=Distortion_MaskCore}
  n33 = OneMinus(None=@n8)
  n34 = MaterialFunctionCall(UVs=TextureCoordinate(), "Rotation Angle (0-1)"=Constant() {r=0.25}) {materialFunction=&CustomRotator}
  n35 = MaterialFunctionCall(Vector or UVs=Subtract(A=Multiply(A=@n34.Rotated Values) {constB=2})) {materialFunction=&VectorToRadialValue}
  n36 = Multiply(A=Subtract(A=ComponentMask(None=@n35.Radial Coordinates) {r=true}) {constB=0.5}) {constB=2}
  n37 = AppendVector(A=LinearInterpolate(A=Multiply(A=@n33) {constB=0.5}, B=Abs(None=@n36), Alpha=@n6.Mask Polar), B=@n30.G)
  n38 = VectorParameter() {
   group=Distortion
   parameterName=Distortion_Main_UV
   channelNames {a=Distortion_V_Speed, b=Distortion_U_Speed, g=Distortion_V_Tile, r=Distortion_U_Tile}
   defaultValue=<rgba 1 0.5 -0.5 0.5>
  }
  n39 = AppendVector(A=@n38.Distortion_U_Tile, B=@n38.Distortion_U_Tile)
  n40 = VectorParameter() {
   group=Distortion
   parameterName=Distortion_Sub_UV
   channelNames {a=Distortion_Sub_V_Speed, b=Distortion_Sub_U_Speed, g=Distortion_Sub_V_Tile, r=Distortion_Sub_U_Tile}
   defaultValue=<rgba 2 1 -0.5 0.3>
  }
  n41 = Panner(
   Coordinate=Multiply(A=TextureCoordinate(), B=AppendVector(A=@n40.Distortion_Sub_U_Tile, B=@n40.Distortion_Sub_U_Tile))
   Speed=AppendVector(A=@n40.Distortion_Sub_U_Speed, B=@n40.Distortion_Sub_U_Speed)
  ) {speedY=0.3}
  n42 = Add(
   A=Panner(Coordinate=Multiply(A=@n37, B=@n39), Speed=AppendVector(A=@n38.Distortion_U_Speed, B=@n38.Distortion_U_Speed))
   B=Multiply(
    A=TextureSample(UVs=@n41) {texture=&T_NoisePack_01}.R
    B=ScalarParameter() {defaultValue=0.35, group=Distortion, parameterName=Distortion_Sub_Intensity}
   ) {constB=0.1}
  )
  n43 = TextureSample(UVs=@n42) {samplerType=SAMPLERTYPE_Normal, texture=/Game/Global/Textures/Default/Tileable_Patterns/T_StormCloudDetail01_N}
  n44 = DynamicParameter() {defaultValue=<rgba 0.65 0 0.45 0.3>, paramNames [Mask High, Mask Thin, Mask Polar, Distortion Intensity]}
  n45 = Add(
   A=@n30.R
   B=Multiply(A=LinearInterpolate(B=@n31, Alpha=@n32) {constA=1, constAlpha=1}, B=Multiply(A=@n43.R, B=@n44.Distortion Intensity)) {constA=1}
  ) {constB=0}
  n46 = CurveAtlasRowParameter(CurveTime=@n45) {
   atlas=&TempAthenaCurveAtlas
   curve=&Curve_Donut_Plate01
   defaultValue=5
  }
  n47 = Multiply(
   A=Multiply(A=@n13.R) {constB=100}
   B=LinearInterpolate(A=LinearInterpolate(A=@n14.RGB, B=@n16, Alpha=@n28), B=LinearInterpolate(A=@n29.RGB, B=@n14.RGB, Alpha=@n28), Alpha=@n46.A)
  )
  n48 = CurveAtlasRowParameter(CurveTime=Saturate(None=@n45)) {
   atlas=&TempAthenaCurveAtlas
   curve=&Curve_Donut_Plate02
   defaultValue=6
   parameterName=DarkEdge
  }
  n49 = VectorParameter() {
   group=Color
   parameterName=Color_Dark_UV
   channelNames {a=Dark_V_Speed, b=Dark_U_Speed, g=Dark_V_Tile, r=Dark_U_Tile}
   defaultValue=<rgba 2 1 -0.3 1>
  }
  n50 = Multiply(A=TextureCoordinate(), B=AppendVector(A=@n49.Dark_U_Tile, B=@n49.Dark_U_Tile))
  n51 = AppendVector(A=@n49.Dark_U_Speed, B=@n49.Dark_U_Speed)
  n52 = Multiply(
   A=@n48.B
   B=TextureSample(UVs=Panner(Coordinate=@n50, Speed=@n51) {speedY=0.1}) {texture=/Game/Global/Textures/Default/Noise/TankMoteNoise}.R
  )
  n53 = LinearInterpolate(A=ScalarParameter() {defaultValue=0.5, group=Color, parameterName=Detail_Brightness}, Alpha=Saturate(None=OneMinus(None=@n52)))
  n54 = ParticleColor()
  MaterialFunctionCall(Emissive=Multiply(A=Multiply(A=@n47, B=@n53), B=@n54.RGB)) {materialFunction=&MF_FX_EmissiveAdjustments}
  n55 = ScalarParameter() {defaultValue=0.4, group=Dissolve, parameterName=Dissolve Distortion Intensity}
  n56 = VectorParameter() {
   group=Dissolve
   parameterName=Dissolve UV Control
   channelNames {a=Dissolve_V_PanSpeed, b=Dissolve_U_PanSpeed, g=Dissolve_V_Tilling, r=Dissolve_U_Tilling}
   defaultValue=<rgba 2 0.5 0 0.5>
  }
  n57 = Multiply(A=Add(A=TextureCoordinate(), B=Multiply(A=@n42, B=@n55) {constB=0.4}), B=AppendVector(A=@n56.Dissolve_U_Tilling, B=@n56.Dissolve_U_Tilling))
  n58 = AppendVector(A=@n56.Dissolve_U_PanSpeed, B=@n56.Dissolve_U_PanSpeed)
  n59 = ScalarParameter() {group=Dissolve, parameterName=Dissolve_U_Offset}
  n60 = ScalarParameter() {group=Dissolve, parameterName=Dissolve_V_Offset}
  n61 = TextureSampleParameter2D(UVs=Add(A=Panner(Coordinate=@n57, Speed=@n58), B=AppendVector(A=@n59, B=@n60))) {
   group=Dissolve
   parameterName=Dissolve Tex
   samplerType=SAMPLERTYPE_Grayscale
   texture=/BRCosmetics/Weapons/FORT_Melee/Pickaxe_Cube_Ninja_Male/FX/T_CellNoise
  }
  n62 = LinearInterpolate(
   A=@n61.G
   B=ComponentMask(None=TextureCoordinate()) {g=true}
   Alpha=ScalarParameter() {defaultValue=0.75, group=Dissolve, parameterName=Lerp Direction}
  ) {constAlpha=0.75}
  n63 = ScalarParameter() {group=Dissolve, parameterName=Use Direction Dissolve}
  n64 = ScalarParameter() {defaultValue=1.1, group=Dissolve, parameterName=Dissolve Edge Hardness, sliderMin=0.5}
  n65 = Multiply(
   A=Power(
    Base=LinearInterpolate(A=@n61.G, B=@n62, Alpha=@n63) {constAlpha=0}
    Exp=ScalarParameter() {defaultValue=0.5, group=Dissolve, parameterName=Dissolve Power}
   )
   B=@n64
  )
  n66 = LinearInterpolate(
   A=@n64
   Alpha=Multiply(
    A=ScalarParameter() {defaultValue=1, group=Dissolve, parameterName=Dissolve Dynamic Test}
    B=DynamicParameter() {parameterIndex=1, paramNames [Dissolve, N/A, N/A, N/A]}.Dissolve
   )
  ) {constB=-1}
  n67 = CurveAtlasRowParameter(CurveTime=Subtract(A=@n45, B=@n2)) {
   atlas=&TempAthenaCurveAtlas
   curve=&Curve_Donut_Plate02
   defaultValue=6
   parameterName=Curve_DonutPlate_2
  }
  n68 = Multiply(A=@n67.A, B=Saturate(None=Subtract(A=@n65, B=@n66)))
  n69 = DepthFade(Opacity=@n68, FadeDistance=ScalarParameter() {defaultValue=2, group=Alpah, parameterName=Fade_Distance}) {fadeDistanceDefault=10}
  Multiply(A=@n54.A, B=ShadingPathSwitch(Default=@n69, Deferred=@n69, Mobile=@n68) {desc=Migrated from FeatureLevelSwitch})
  Reroute(None=@n68)
  ScalarParameter() {group=Mask, parameterName=Mask_Thin}
  ScalarParameter() {defaultValue=0.3, group=Distortion, parameterName=Distortion_Intensity}
  ScalarParameter() {defaultValue=0.45, group=Mask, parameterName=Mask_Polar}
  ScalarParameter() {defaultValue=0.65, group=Mask, parameterName=Mask_High, sortPriority=1}
 ]
}
