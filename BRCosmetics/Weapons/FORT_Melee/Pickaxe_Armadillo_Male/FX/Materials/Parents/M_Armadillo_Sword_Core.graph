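# Translucent unlit parent material for Niagara mesh particles in the Armadillo pickaxe's sword core effect, drawing swirling storm smoke with glowing edges.
# Two rotating, panning cloud masks and normals blend by per-particle randoms.
# A particle timestamp drives an eased spin burst and scales extra glow.
# Normal-versus-camera and fresnel terms shape color and depth-faded opacity, with a half-sphere mask alternative, time-of-day emissive scaling and noise-driven vertex offset.
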Material /BRCosmetics/Weapons/FORT_Melee/Pickaxe_Armadillo_Male/FX/Materials/Parents/M_Armadillo_Sword_Core
graph {
 properties {
  bUsedWithNiagaraMeshParticles=true
  blendMode=BLEND_Translucent
  pixelDepthOffsetMode=PDOM_Legacy
  previewMesh=/Game/Effects/Fort_Effects/UI/LevelUp/SM_Geosphere
  shadingModel=MSM_Unlit
  usedShadingModels=Unlit
 }
 nodes [
  n1 = TextureCoordinate()
  n2 = VectorParameter() {parameterName=Tile, defaultValue=<rgba 1 1 0 0>}
  n3 = AppendVector(A=@n2.R, B=@n2.R)
  n4 = DynamicParameter() {paramNames [Random,PanSpeed,TimeStamp,Offset]}
  n5 = Subtract(A=Time() {period=5}, B=@n4.TimeStamp)
  n6 = ScalarParameter() {defaultValue=3, group=Timestamp, parameterName=SpinDuration}
  n7 = MaterialFunctionCall!(In=Clamp(None=Divide(A=@n5, B=@n6))) {materialFunction=/Game/Effects/Fort_Effects/Materials/_Functions/MF_Smoothstep_EaseOut}
  n8 = ScalarParameter() {defaultValue=-3, parameterName=SpinAmount}
  n9 = Time()
  n10 = Multiply(A=ScalarParameter() {parameterName=RotationSpeed}, B=Add(A=Multiply(A=@n7.Result, B=@n8) {constB=-18}, B=@n9))
  n11 = Multiply(A=@n7.Result) {constB=-2}
  n12 = Add(A=Multiply(A=Add(A=@n11, B=@n9), B=@n4.PanSpeed), B=@n4.Random)
  n13 = Panner(Coordinate=Rotator(Coordinate=Multiply(A=@n1, B=@n3), Time=@n10) {centerX=0.4, centerY=0.4, speed=1}, Time=@n12) {speedX=0.05, speedY=0.07}
  n14 = TextureSampleParameter2D(UVs=@n13) {parameterName=Mask1, samplerType=SAMPLERTYPE_LinearGrayscale, texture=/Game/Environments/Hexmap/HexClouds/T_HexmapFog_01_Height_GS}
  n15 = TextureCoordinate()
  n16 = Panner(Coordinate=Rotator(Coordinate=Multiply(A=@n3, B=@n15), Time=@n10) {centerX=0.2, centerY=0.2, speed=1}, Time=@n12) {speedX=-0.01, speedY=0.06}
  n17 = DynamicParameter() {parameterIndex=1, paramNames [RandomAlphaLerp, RandomPower, Param3, Param4]}
  n18 = LinearInterpolate(
   A=@n14.R
   B=TextureSampleParameter2D(UVs=@n16) {parameterName=Mask2, samplerType=SAMPLERTYPE_LinearGrayscale, texture=/Game/Global/Textures/Default/Flowmaps/T_StormSwirlMask}.R
   Alpha=@n17.RandomAlphaLerp
  ) {constAlpha=0.4}
  n19 = Clamp(None=LinearInterpolate(Alpha=@n18) {constA=-0.1})
  n20 = TextureSampleParameter2D(UVs=@n13) {parameterName=Normal1, samplerType=SAMPLERTYPE_Normal, texture=/Game/Environments/Hexmap/HexClouds/T_HexmapFog_01_N}
  n21 = TextureSampleParameter2D(UVs=@n16) {parameterName=Normal2, samplerType=SAMPLERTYPE_Normal, texture=/Game/VisualThreatManager/StormVisuals/LargeClouds/T_StormSwirlNormal}
  n22 = ShadingPathSwitch(Default=LinearInterpolate(A=@n20.RGB, B=@n21.RGB, Alpha=@n17.RandomAlphaLerp) {constAlpha=0.4}, Mobile=@n20.RGB) {desc=Migrated from FeatureLevelSwitch}
  n23 = Saturate(None=DotProduct(A=Transform(None=@n22), B=CameraVectorWS()))
  n24 = Multiply(A=@n23, B=@n23)
  n25 = Multiply(A=Multiply(A=@n19, B=Power(Base=@n24, Exp=@n17.RandomPower)), B=ScalarParameter() {defaultValue=4, parameterName=OpacityMult}) {constB=4}
  n26 = ScalarParameter() {defaultValue=2, parameterName=CloudOpacityClampMax}
  n27 = ScalarParameter() {defaultValue=1, parameterName=SmokeAlphaFresnelExp}
  n28 = Transform(None=MaterialFunctionCall(Normal=@n22, Flatness=ScalarParameter() {defaultValue=0.7, parameterName=NormalFlatness}) {materialFunction=&FlattenNormal}.Result)
  n29 = Multiply(
   A=Multiply(
    A=Fresnel(ExponentIn=@n27, Normal=@n28) {baseReflectFraction=0, exponent=0.1}
    B=ScalarParameter() {defaultValue=1, parameterName=SmokeAlphaMult}
   )
   B=Multiply(A=Clamp(None=@n25, Max=@n26) {maxDefault=2}, B=VertexColor().R)
  )
  n30 = ComponentMask(None=TextureCoordinate()) {g=true}
  n31 = Saturate(None=Subtract(A=Power(Base=@n30)) {constB=0.025})
  n32 = Multiply(A=Multiply(A=Multiply(A=@n19, B=@n31), B=Fresnel() {baseReflectFraction=0, exponent=1})) {constB=10}
  n33 = ScalarParameter() {parameterName=Half Sphere Mask}
  DepthFade(
   Opacity=Multiply(A=ParticleColor().A, B=LinearInterpolate(A=@n29, B=@n32, Alpha=@n33) {constA=1})
   FadeDistance=ScalarParameter() {parameterName=DepthFadeDist}
  )
  Divide(A=Time() {bOverride_Period=true, period=3}, B=@n6) {constB=3, desc=For testing}
  n34 = ScalarParameter() {defaultValue=1, group=Time of Day-Color, parameterName=ToD_Morning}
  n35 = ScalarParameter() {defaultValue=1, group=Time of Day-Color, parameterName=ToD_Day}
  n36 = ScalarParameter() {defaultValue=1, group=Time of Day-Color, parameterName=ToD_Evening}
  n37 = ScalarParameter() {defaultValue=1, group=Time of Day-Color, parameterName=ToD_Night}
  n38 = ParticleColor()
  n39 = VectorParameter() {parameterName=SmokeColor1, defaultValue=<rgba 0.0484 0.0653 0.1615 0>}
  n40 = LinearInterpolate(A=@n39.RGB, B=VectorParameter() {parameterName=SmokeColor2, defaultValue=<rgba 0.026 0.0274 0.04 0>}.RGB, Alpha=@n18)
  n41 = ScalarParameter() {defaultValue=0.1, parameterName=SmokeColorFresnelExp}
  n42 = Clamp(None=LinearInterpolate(Alpha=Fresnel(ExponentIn=@n41, Normal=@n28) {baseReflectFraction=0, exponent=0.1}) {constA=1.7, constB=-0.25})
  n43 = Multiply(A=@n39.RGB, B=Multiply(A=Multiply(A=@n42, B=@n18), B=ScalarParameter() {defaultValue=4, parameterName=SmokeColor1Mult}) {constB=4})
  n44 = VectorParameter() {parameterName=EmissiveColor, defaultValue=<rgba 0.8921 0.306 1.25 0>}
  n45 = Multiply(A=ScalarParameter() {parameterName=TimeStampParamForGlow}, B=@n4.TimeStamp)
  n46 = Power(Base=@n23, Exp=ScalarParameter() {defaultValue=4, parameterName=EdgePower}) {constExponent=4}
  n47 = Multiply(A=@n45, B=Add(A=@n46) {constB=0.05})
  n48 = Clamp(None=LinearInterpolate(
    A=ScalarParameter() {defaultValue=1, parameterName=EdgeGlowMax}
    B=ScalarParameter() {defaultValue=-0.5, parameterName=EdgeGlowMin}
    Alpha=@n18
   ) {constA=1, constB=-0.3})
  n49 = Add(A=Add(A=Add(A=@n40, B=@n43), B=Multiply(A=@n44.RGB, B=@n47)), B=Multiply(A=@n44.RGB, B=Multiply(A=@n48, B=@n46)))
  MaterialFunctionCall(Emissive=Multiply(
    A=MaterialFunctionCall(Value_Morning=@n34, Value_Day=@n35, Value_Evening=@n36, Value_Night=@n37) {materialFunction=&Scalar_Time_of_Day}.Result
    B=Multiply(A=@n38.RGB, B=@n49)
   )) {materialFunction=&MF_FX_EmissiveAdjustments}
  n50 = VertexNormalWS()
  Multiply(A=Multiply(A=Normalize(VectorInput=@n50), B=LinearInterpolate(Alpha=@n18) {constA=-0.5}), B=ScalarParameter() {parameterName=WPO})
  Multiply(A=Multiply(A=@n45) {constB=0.25}, B=@n44.RGB)
 ]
}
